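; Spline mesh actor that blends the landscape to its shape through a LandscapeTexturePatch, with custom events to reset or conform the spline to terrain.
; UpdateLandscapePatch sizes the patch to the spline, then scene-captures temporary blend meshes and end pieces into height, min-height and alpha render targets.
; A composite material draws these into the height patch; optional weight patches paint left/right, top/bottom layers.
; EnableFastEditing disables the patch and skips this rebuild.
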(event Custom|CaptureTerrainHeight
  (CallFunction|CaptureLandscapeHeight))

(event Custom|1%20ResetTerrain
  (LandscapePatch|SetIsEnabled (Variables|Default|GetLandscapeTexturePatch) false))

(event Custom|2%20ConformSpline%20to%20Landscape
  (LandscapePatch|SetIsEnabled (Variables|Default|GetLandscapeTexturePatch) false)
  (Default|ConformSplinetoLandscape)
  (CallFunction|UpdateLandscapePatch))

(event EventBeginPlay)

(event Collision|EventActorBeginOverlap
  (params (OtherActor :type Actor%20Object%20Reference)))

(event EventTick
  (params (DeltaSeconds :type Float%20%28single-precision%29)))

(event Custom|UpdateLandscapePatch
  (bind _staticmesh (Variables|Default|GetStaticMesh))
  (bind _returnvalue (StaticMesh|GetBoundingBox _staticmesh))
  (bind (_min _max) (Utilities|Struct|BreakBox _returnvalue))
  (bind _returnvalue_2 (Math|Color|MakeColor 1.0))
  (bind (_x _y) (Math|Vector2D|BreakVector2D (LandscapePatch|GetInitResolutionfromLandscape (Variables|Default|GetLandscapeTexturePatch) 1.0)))
  (bind _y_2 (.y _min))
  (bind _y_3 (.y _max))
  (bind _landscapetexturepatch (Variables|Default|GetLandscapeTexturePatch))
  (bind _resolutionout (LandscapePatch|GetInitResolutionfromLandscape _landscapetexturepatch 1.0))
  (bind (_x_2 _y_4) (Math|Vector2D|BreakVector2D _resolutionout))
  (bind _returnvalue_3 (Math|Float|Truncate _x_2))
  (bind _returnvalue_4 (Math|Float|Truncate _y_4))
  (bind _returnvalue_5 (Math|Color|MakeColor 0.0 1.0))
  (bind _self self)
  (bind _array (Utilities|Array|MakeArray _self))
  (bind _layer_right_bottom (Variables|LandscapeWeightPainting|GetLayer%20Right_Bottom))
  (bind _layer_right_top (Variables|LandscapeWeightPainting|GetLayer%20Right_Top))
  (bind _returnvalue_6 (== _layer_right_top _layer_right_bottom))
  (bind _layer_left_bottom (Variables|LandscapeWeightPainting|GetLayer%20Left_Bottom))
  (bind _layer_right_top_2 (Variables|LandscapeWeightPainting|GetLayer%20Right_Top))
  (bind _returnvalue_7 (== _layer_right_top_2 _layer_left_bottom))
  (bind _layer_left_bottom_2 (Variables|LandscapeWeightPainting|GetLayer%20Left_Bottom))
  (bind _layer_right_bottom_2 (Variables|LandscapeWeightPainting|GetLayer%20Right_Bottom))
  (bind _returnvalue_8 (== _layer_right_bottom_2 _layer_left_bottom_2))
  (bind _layer_right_bottom_3 (Variables|LandscapeWeightPainting|GetLayer%20Right_Bottom))
  (bind _layer_left_bottom_3 (Variables|LandscapeWeightPainting|GetLayer%20Left_Bottom))
  (bind _returnvalue_9 (== _layer_left_bottom_3 _layer_right_bottom_3))
  (bind _returnvalue_10 (- _y_3 _y_2))
  (bind _size_y (Variables|Default|GetSize%20Y))
  (bind _noisecontrast (Variables|LandscapeWeightPainting|GetNoiseContrast))
  (bind _noisecontrast_2 (Variables|LandscapeWeightPainting|GetNoiseContrast))
  (bind _scenecapturecomponent2d (Variables|Default|GetSceneCaptureComponent2D))
  (bind _height_cutoff_distance (Variables|TerrainBlend|GetHeight%20Cutoff%20Distance))
  (bind _blend_distance (Variables|TerrainBlend|GetBlend%20Distance))
  (bind _errormargin (Variables|TerrainBlend|GetErrorMargin))
  (bind _patchmargin (Variables|TerrainBlend|GetPatchMargin))
  (bind _scenecapturecomponent2d_2 (Variables|Default|GetSceneCaptureComponent2D))
  (bind _alpha_falloff (Variables|TerrainBlend|GetAlpha%20Falloff))
  (bind _blend_distance_2 (Variables|TerrainBlend|GetBlend%20Distance))
  (bind _errormargin_2 (Variables|TerrainBlend|GetErrorMargin))
  (bind _min_height_offset_blend (Variables|TerrainBlend|GetMin%20Height%20Offset%20Blend))
  (bind _size_x (Variables|Default|GetSize%20X))
  (if (Variables|TerrainBlend|GetEnableFastEditing)
    (LandscapePatch|SetIsEnabled (Variables|Default|GetLandscapeTexturePatch) false)
    (else
      (Utilities|FlowControl|Sequence
        (:then_0
          (LandscapePatch|SetIsEnabled (Variables|Default|GetLandscapeTexturePatch) true)
          (bind (_x_max _x_min _y_max _y_min _size_x_2 _size_y_2 _size_z _origin_local _origin_world _scale_y_max) (CallFunction|GetSplineMeshBounds (Variables|Default|GetSpline)))
          (Variables|Default|SetSize%20X _size_x_2)
          (Variables|Default|SetSize%20Y _size_y_2)
          (bind _maxboundwidth (CallFunction|GetStartEndBounds))
          (CallFunction|SetupLandscapePatch (Math|Vector2D|MakeVector2D (Math|Float|Max(Float) (Utilities|Operators|Add _size_x (* _returnvalue_10 _scale_y_max) _patchmargin _blend_distance _maxboundwidth) (Utilities|Operators|Add _size_y (* _returnvalue_10 _scale_y_max) _patchmargin _blend_distance _maxboundwidth)) (Math|Float|Max(Float) (Utilities|Operators|Add _size_x (* _returnvalue_10 _scale_y_max) _patchmargin _blend_distance _maxboundwidth) (Utilities|Operators|Add _size_y (* _returnvalue_10 _scale_y_max) _patchmargin _blend_distance _maxboundwidth))) _origin_world)
          (Transformation|SetWorldLocation _scenecapturecomponent2d (+ _origin_world (Math|Vector|MakeVector 0.0 0.0 (Utilities|Operators|Add _size_z (.z _max) _blend_distance _patchmargin (Variables|TerrainBlend|GetHeight%20Cutoff%20Distance)))) false false)
          (Transformation|SetRelativeRotation _scenecapturecomponent2d "-90.000000,-90.000000,0.000000" false false)
          (Utilities|FlowControl|Sequence
            (:then_1
              (bind _returnvalue_11 (Rendering|CreateRenderTarget2D _returnvalue_3 _returnvalue_4 "RTF_R32f"))
              (Variables|/Script/Engine.SceneCaptureComponent2D|SetTextureTarget _returnvalue_11 _scenecapturecomponent2d_2) ; RenderTarget for Object height
              (bind _returnvalue_12 (Rendering|CreateRenderTarget2D _returnvalue_3 _returnvalue_4 "RTF_R32f"))
              (Variables|/Script/Engine.SceneCaptureComponent2D|SetTextureTarget _returnvalue_12 (Variables|Default|GetSceneCaptureComponent2DMIN))
              (bind _returnvalue_13 (Rendering|CreateRenderTarget2D _returnvalue_3 _returnvalue_4 "RTF_R8"))
              (Variables|/Script/Engine.SceneCaptureComponent2D|SetTextureTarget _returnvalue_13 (Variables|Default|GetSceneCaptureComponent2DALPHA))
              (Variables|/Script/Engine.SceneCaptureComponent2D|SetOrthoWidth (Math|Float|Max(Float) (Utilities|Operators|Add _size_x (* _returnvalue_10 _scale_y_max) _patchmargin _blend_distance _maxboundwidth) (Utilities|Operators|Add _size_y (* _returnvalue_10 _scale_y_max) _patchmargin _blend_distance _maxboundwidth)) (Variables|Default|GetSceneCaptureComponent2D))
              (Variables|/Script/Engine.SceneCaptureComponent2D|SetOrthoWidth (Math|Float|Max(Float) (Utilities|Operators|Add _size_x (* _returnvalue_10 _scale_y_max) _patchmargin _blend_distance _maxboundwidth) (Utilities|Operators|Add _size_y (* _returnvalue_10 _scale_y_max) _patchmargin _blend_distance _maxboundwidth)) (Variables|Default|GetSceneCaptureComponent2DALPHA))
              (Variables|/Script/Engine.SceneCaptureComponent2D|SetOrthoWidth (Math|Float|Max(Float) (Utilities|Operators|Add _size_x (* _returnvalue_10 _scale_y_max) _patchmargin _blend_distance _maxboundwidth) (Utilities|Operators|Add _size_y (* _returnvalue_10 _scale_y_max) _patchmargin _blend_distance _maxboundwidth)) (Variables|Default|GetSceneCaptureComponent2DMIN))
              (Variables|/Script/Engine.SceneCaptureComponent|SetShowOnlyActors (Utilities|Array|MakeArray self) _scenecapturecomponent2d_2)
              (Utilities|FlowControl|Sequence
                (:then_0
                  (for _array_element (Variables|Default|GetAllSplineComponents)
                    (Rendering|LOD|SetForcedLodModel _array_element 1)))
                (:then_1
                  (bind _allsplinecomponents (CallFunction|AddSplineMeshComponentsforBlend true "/Game/ArtTools/Procedural/LandscapeBlend/SplineMeshBlend/SplinemeshLandscapeBlend.SplinemeshLandscapeBlend" _min_height_offset_blend _alpha_falloff _blend_distance_2 _errormargin_2))
                  (bind _end_pieceslist (CallFunction|AddEndBlendComponents _min_height_offset_blend _errormargin_2 "(R=1.000000,G=1.000000,B=1.000000,A=1.000000)" "(R=1.000000,G=1.000000,B=1.000000,A=1.000000)" _alpha_falloff))
                  (Rendering|SceneCapture|CaptureScene (Variables|Default|GetSceneCaptureComponent2D) :!class /Script/Engine.SceneCaptureComponent2D)
                  (if (not (Variables|TerrainBlend|GetDebugBlendMesh))
                    (for _array_element_2 _allsplinecomponents
                      (Components|DestroyComponent _array_element_2))
                    (for _array_element_3 _end_pieceslist
                      (Components|DestroyComponent _array_element_3))))
                (:then_2
                  (bind _allsplinecomponents_2 (CallFunction|AddSplineMeshComponentsforBlend true "/Game/ArtTools/Procedural/LandscapeBlend/SplineMeshBlend/SplinemeshLandscapeBlend.SplinemeshLandscapeBlend" _height_cutoff_distance _alpha_falloff _blend_distance_2))
                  (bind _end_pieceslist_2 (CallFunction|AddEndBlendComponents _height_cutoff_distance 0.0 "(R=1.000000,G=1.000000,B=1.000000,A=1.000000)" "(R=1.000000,G=1.000000,B=1.000000,A=1.000000)" _alpha_falloff))
                  (CallFunction|SetVisibilityonArray (Variables|Default|GetAllSplineComponents))
                  (CallFunction|SetStartEndComponentVisiblity false)
                  (Variables|/Script/Engine.SceneCaptureComponent|SetShowOnlyActors _array (Variables|Default|GetSceneCaptureComponent2DALPHA))
                  (Variables|/Script/Engine.SceneCaptureComponent|SetShowOnlyActors _array (Variables|Default|GetSceneCaptureComponent2DMIN))
                  (Rendering|SceneCapture|CaptureScene (Variables|Default|GetSceneCaptureComponent2DALPHA) :!class /Script/Engine.SceneCaptureComponent2D)
                  (Rendering|SceneCapture|CaptureScene (Variables|Default|GetSceneCaptureComponent2DMIN) :!class /Script/Engine.SceneCaptureComponent2D)
                  (CallFunction|SetVisibilityonArray (Variables|Default|GetAllSplineComponents) true)
                  (CallFunction|SetStartEndComponentVisiblity true))
                (:then_3
                  (Utilities|FlowControl|Sequence
                    (:then_0
                      (for _array_element_4 (Variables|Default|GetAllSplineComponents)
                        (Rendering|LOD|SetForcedLodModel _array_element_4 0)))
                    (:then_2
                      (if (not (Variables|TerrainBlend|GetDebugBlendMesh))
                        (for _array_element_5 _allsplinecomponents_2
                          (Components|DestroyComponent _array_element_5))
                        (for _array_element_6 _end_pieceslist_2
                          (Components|DestroyComponent _array_element_6))))))))
            (:then_2
              ;[1 1 1] WriteMaterial To HeightPatch
              (bind _returnvalue_14 (Rendering|Material|CreateDynamicMaterialInstance "/Game/ArtTools/Procedural/LandscapeBlend/SplineMeshBlend/M_LandscapeHeightComposite.M_LandscapeHeightComposite" "None" "None" :!class /Script/Engine.KismetMaterialLibrary))
              (bind _output_get (Variables|Default|SetMID _returnvalue_14))
              (Rendering|Material|SetTextureParameterValue _output_get "RenderTargetObject" _returnvalue_11)
              (Rendering|Material|SetTextureParameterValue _output_get "RenderTargetMinHeight" _returnvalue_12)
              (Rendering|Material|SetTextureParameterValue _output_get "RenderTargetAlpha" _returnvalue_13)
              (Rendering|Material|SetScalarParameterValue _output_get "Z_PositionCamera" (.z (Transformation|GetWorldLocation (Variables|Default|GetSceneCaptureComponent2D))) :!class /Script/Engine.MaterialInstanceDynamic)
              (bind _returnvalue_15 (CallFunction|GetClosestLandscape))
              (bind _outactors (Actor|GetAllActorsOfClass "/Script/Landscape.Landscape"))
              (Rendering|Material|SetScalarParameterValue _output_get "Z_PositionTerrain" (.z (Transformation|GetActorLocation _returnvalue_15)) :!class /Script/Engine.MaterialInstanceDynamic)
              (bind _returnvalue_16 (LandscapePatch|GetHeightRenderTarget (Variables|Default|GetLandscapeTexturePatch) true))
              (Rendering|DrawMaterialtoRenderTarget _returnvalue_16 (Variables|Default|GetMID)))
            (:then_3
              (if (or (Variables|LandscapeWeightPainting|GetWeightPainting%20Right) (Variables|LandscapeWeightPainting|GetWeightPainting%20Left))
                (bind _returnvalue_17 (Rendering|CreateRenderTarget2D (Math|Float|Truncate _x) (Math|Float|Truncate _y) "RTF_RGBA8"))
                (Variables|/Script/Engine.SceneCaptureComponent2D|SetTextureTarget _returnvalue_17 (Variables|Default|GetSceneCaptureComponent2DALPHA))
                (Utilities|FlowControl|Sequence
                  (:then_0
                    (bind _allsplinecomponents_3 (CallFunction|AddSplineMeshComponentsforBlend true "/Game/ArtTools/Procedural/LandscapeBlend/SplineMeshBlend/SplinemeshLandscapeBlend.SplinemeshLandscapeBlend" 1e+02 1.0 (Variables|TerrainBlend|GetBlend%20Distance) _errormargin _returnvalue_2 _returnvalue_5))
                    (bind _end_pieceslist_3 (CallFunction|AddEndBlendComponents 1e+02 _errormargin _returnvalue_2 _returnvalue_5 1.0))
                    (CallFunction|SetVisibilityonArray (Variables|Default|GetAllSplineComponents))
                    (CallFunction|SetStartEndComponentVisiblity false)
                    (Rendering|SceneCapture|CaptureScene (Variables|Default|GetSceneCaptureComponent2DALPHA) :!class /Script/Engine.SceneCaptureComponent2D)
                    (CallFunction|SetVisibilityonArray (Variables|Default|GetAllSplineComponents) true)
                    (CallFunction|SetStartEndComponentVisiblity true)
                    (if (not false)
                      (for _array_element_7 _allsplinecomponents_3
                        (Components|DestroyComponent _array_element_7))
                      (for _array_element_8 _end_pieceslist_3
                        (Components|DestroyComponent _array_element_8))))
                  (:then_1
                    (if (Variables|LandscapeWeightPainting|GetWeightPainting%20Right)
                      ;[1 1 1] Weight Patch Render Target
                      (Utilities|FlowControl|Sequence
                        (:then_0
                          (bind _returnvalue_18 (Rendering|Material|CreateDynamicMaterialInstance "/Game/ArtTools/Procedural/LandscapeBlend/SplineMeshBlend/M_Landscape_Weight_Paint.M_Landscape_Weight_Paint" "None" "None" :!class /Script/Engine.KismetMaterialLibrary))
                          (Rendering|Material|SetTextureParameterValue _returnvalue_18 "RenderTargetAlpha" _returnvalue_17)
                          (Rendering|Material|SetScalarParameterValue _returnvalue_18 "WorldSize" (Math|Vector2D|BreakVector2D (LandscapePatch|GetUnscaledCoverage (Variables|Default|GetLandscapeTexturePatch))) :!class /Script/Engine.MaterialInstanceDynamic)
                          (Rendering|Material|SetScalarParameterValue _returnvalue_18 "Red" 1.0 :!class /Script/Engine.MaterialInstanceDynamic)
                          (Rendering|Material|SetScalarParameterValue _returnvalue_18 "isInvertedRed" 0.0 :!class /Script/Engine.MaterialInstanceDynamic)
                          (Rendering|Material|SetScalarParameterValue _returnvalue_18 "Green" (Math|Conversions|ToFloat(Boolean) (and (or (== _layer_right_top_2 (Variables|LandscapeWeightPainting|GetLayer%20Left_Top)) _returnvalue_7) (Variables|LandscapeWeightPainting|GetWeightPainting%20Left))) :!class /Script/Engine.MaterialInstanceDynamic)
                          (Rendering|Material|SetScalarParameterValue _returnvalue_18 "isInvertedGreen" (Math|Conversions|ToFloat(Boolean) _returnvalue_7) :!class /Script/Engine.MaterialInstanceDynamic)
                          (Rendering|Material|SetScalarParameterValue _returnvalue_18 "NoiseStrengthRed" (Math|Float|SelectFloat 0.0 1.0 (== (Variables|LandscapeWeightPainting|GetLayer%20Right_Top) (Variables|LandscapeWeightPainting|GetLayer%20Right_Bottom))) :!class /Script/Engine.MaterialInstanceDynamic)
                          (Rendering|Material|SetScalarParameterValue _returnvalue_18 "NoiseStrengthGreen" (Math|Float|SelectFloat 0.0 1.0 (== (Variables|LandscapeWeightPainting|GetLayer%20Left_Top) (Variables|LandscapeWeightPainting|GetLayer%20Left_Bottom))) :!class /Script/Engine.MaterialInstanceDynamic)
                          (Rendering|Material|SetScalarParameterValue _returnvalue_18 "Contrast" _noisecontrast :!class /Script/Engine.MaterialInstanceDynamic))
                        (:then_1
                          (LandscapePatch|CreateWeightPatch (Variables|Default|GetLandscapeTexturePatch) (Variables|LandscapeWeightPainting|GetLayer%20Right_Top) "TextureBackedRenderTarget" "SourceTextureChannel")
                          (bind _returnvalue_19 (LandscapePatch|GetWeightPatchRenderTarget (Variables|Default|GetLandscapeTexturePatch) (Variables|LandscapeWeightPainting|GetLayer%20Right_Top) true))
                          (Rendering|DrawMaterialtoRenderTarget _returnvalue_19 _returnvalue_18)
                          (if (not (== (Variables|LandscapeWeightPainting|GetLayer%20Right_Bottom) ""))
                            (LandscapePatch|CreateWeightPatch (Variables|Default|GetLandscapeTexturePatch) (Variables|LandscapeWeightPainting|GetLayer%20Right_Bottom) "TextureBackedRenderTarget" "SourceTextureChannel")
                            (bind _returnvalue_20 (LandscapePatch|GetWeightPatchRenderTarget (Variables|Default|GetLandscapeTexturePatch) (Variables|LandscapeWeightPainting|GetLayer%20Right_Bottom) true))
                            (Rendering|Material|SetScalarParameterValue _returnvalue_18 "Green" (Math|Conversions|ToFloat(Boolean) (and (or (== _layer_right_bottom_2 (Variables|LandscapeWeightPainting|GetLayer%20Left_Top)) _returnvalue_8) (Variables|LandscapeWeightPainting|GetWeightPainting%20Left))) :!class /Script/Engine.MaterialInstanceDynamic)
                            (Rendering|Material|SetScalarParameterValue _returnvalue_18 "isInvertedRed" 1.0 :!class /Script/Engine.MaterialInstanceDynamic)
                            (Rendering|Material|SetScalarParameterValue _returnvalue_18 "isInvertedGreen" (Math|Conversions|ToFloat(Boolean) _returnvalue_8) :!class /Script/Engine.MaterialInstanceDynamic)
                            (Rendering|Material|SetScalarParameterValue _returnvalue_18 "Contrast" _noisecontrast :!class /Script/Engine.MaterialInstanceDynamic)
                            (Rendering|Material|SetScalarParameterValue _returnvalue_18 "NoiseStrengthRed" 0.0 :!class /Script/Engine.MaterialInstanceDynamic)
                            (Rendering|Material|SetScalarParameterValue _returnvalue_18 "NoiseStrengthGreen" 0.0 :!class /Script/Engine.MaterialInstanceDynamic)
                            (Rendering|DrawMaterialtoRenderTarget _returnvalue_20 _returnvalue_18))))))
                  (:then_2
                    (if (Variables|LandscapeWeightPainting|GetWeightPainting%20Left)
                      ;[1 1 1] Weight Patch Render Target
                      (Utilities|FlowControl|Sequence
                        (:then_0
                          (bind _returnvalue_21 (Rendering|Material|CreateDynamicMaterialInstance "/Game/ArtTools/Procedural/LandscapeBlend/SplineMeshBlend/M_Landscape_Weight_Paint.M_Landscape_Weight_Paint" "None" "None" :!class /Script/Engine.KismetMaterialLibrary))
                          (Rendering|Material|SetTextureParameterValue _returnvalue_21 "RenderTargetAlpha" _returnvalue_17)
                          (Rendering|Material|SetScalarParameterValue _returnvalue_21 "WorldSize" (Math|Vector2D|BreakVector2D (LandscapePatch|GetUnscaledCoverage (Variables|Default|GetLandscapeTexturePatch))) :!class /Script/Engine.MaterialInstanceDynamic)
                          (Rendering|Material|SetScalarParameterValue _returnvalue_21 "Green" 1.0 :!class /Script/Engine.MaterialInstanceDynamic)
                          (Rendering|Material|SetScalarParameterValue _returnvalue_21 "isInvertedGreen" 0.0 :!class /Script/Engine.MaterialInstanceDynamic)
                          (Rendering|Material|SetScalarParameterValue _returnvalue_21 "Red" (Math|Conversions|ToFloat(Boolean) (and (or (== (Variables|LandscapeWeightPainting|GetLayer%20Left_Top) _layer_right_top) _returnvalue_6) (Variables|LandscapeWeightPainting|GetWeightPainting%20Right))) :!class /Script/Engine.MaterialInstanceDynamic)
                          (Rendering|Material|SetScalarParameterValue _returnvalue_21 "isInvertedRed" (Math|Conversions|ToFloat(Boolean) _returnvalue_6) :!class /Script/Engine.MaterialInstanceDynamic)
                          (Rendering|Material|SetScalarParameterValue _returnvalue_21 "NoiseStrengthGreen" (Math|Float|SelectFloat 0.0 1.0 (== (Variables|LandscapeWeightPainting|GetLayer%20Left_Top) (Variables|LandscapeWeightPainting|GetLayer%20Left_Bottom))) :!class /Script/Engine.MaterialInstanceDynamic)
                          (Rendering|Material|SetScalarParameterValue _returnvalue_21 "NoiseStrengthRed" (Math|Float|SelectFloat 0.0 1.0 (== (Variables|LandscapeWeightPainting|GetLayer%20Right_Top) (Variables|LandscapeWeightPainting|GetLayer%20Right_Bottom))) :!class /Script/Engine.MaterialInstanceDynamic)
                          (Rendering|Material|SetScalarParameterValue _returnvalue_21 "Contrast" _noisecontrast_2 :!class /Script/Engine.MaterialInstanceDynamic))
                        (:then_1
                          (LandscapePatch|CreateWeightPatch (Variables|Default|GetLandscapeTexturePatch) (Variables|LandscapeWeightPainting|GetLayer%20Left_Top) "TextureBackedRenderTarget" "SourceTextureChannel")
                          (bind _returnvalue_22 (LandscapePatch|GetWeightPatchRenderTarget (Variables|Default|GetLandscapeTexturePatch) (Variables|LandscapeWeightPainting|GetLayer%20Left_Top) true))
                          (Rendering|DrawMaterialtoRenderTarget _returnvalue_22 _returnvalue_21)
                          (if (not (== (Variables|LandscapeWeightPainting|GetLayer%20Left_Bottom) ""))
                            (LandscapePatch|CreateWeightPatch (Variables|Default|GetLandscapeTexturePatch) (Variables|LandscapeWeightPainting|GetLayer%20Left_Bottom) "TextureBackedRenderTarget" "SourceTextureChannel")
                            (bind _returnvalue_23 (LandscapePatch|GetWeightPatchRenderTarget (Variables|Default|GetLandscapeTexturePatch) (Variables|LandscapeWeightPainting|GetLayer%20Left_Bottom) true))
                            (Rendering|Material|SetScalarParameterValue _returnvalue_21 "Red" (Math|Conversions|ToFloat(Boolean) (and (or (== _layer_left_bottom_3 (Variables|LandscapeWeightPainting|GetLayer%20Right_Top)) _returnvalue_9) (Variables|LandscapeWeightPainting|GetWeightPainting%20Right))) :!class /Script/Engine.MaterialInstanceDynamic)
                            (Rendering|Material|SetScalarParameterValue _returnvalue_21 "isInvertedGreen" 1.0 :!class /Script/Engine.MaterialInstanceDynamic)
                            (Rendering|Material|SetScalarParameterValue _returnvalue_21 "isInvertedRed" (Math|Conversions|ToFloat(Boolean) _returnvalue_9) :!class /Script/Engine.MaterialInstanceDynamic)
                            (Rendering|Material|SetScalarParameterValue _returnvalue_21 "Contrast" _noisecontrast_2 :!class /Script/Engine.MaterialInstanceDynamic)
                            (Rendering|Material|SetScalarParameterValue _returnvalue_21 "NoiseStrengthGreen" 0.0 :!class /Script/Engine.MaterialInstanceDynamic)
                            (Rendering|Material|SetScalarParameterValue _returnvalue_21 "NoiseStrengthRed" 0.0 :!class /Script/Engine.MaterialInstanceDynamic)
                            (Rendering|DrawMaterialtoRenderTarget _returnvalue_23 _returnvalue_21)))))))))))))))

(fn Add_SplineMesh_Components
  (params (Evaluate%20WPO :type Boolean) (Mesh :type Static%20Mesh%20Object%20Reference) (Material :type Material%20Interface%20Object%20Reference))
  (bind _localsplinecomponentlist (LocalVariables|GetLocalSplineComponentlist))
  (Utilities|FlowControl|Sequence
    (:then_0
      (for _index (range (Spline|GetNumberOfSplinePoints (Variables|Default|GetSpline)))
        (LocalVariables|SetCurrentPoint _index)
        (bind _returnvalue (AddComponent|NODE_AddSplineMeshComponent-0 :RelativeTransform (Math|Transform|MakeTransform "0, 0, 0")))
        (Rendering|LOD|SetEvaluateWorldPositionOffset _returnvalue Evaluate%20WPO :!class /Script/Engine.StaticMeshComponent)
        (Default|SetSplineMeshCollision _returnvalue)
        (bind _output_get (LocalVariables|SetCurrentSplineComponent _returnvalue))
        (Utilities|Array|Add _localsplinecomponentlist _output_get)
        (Components|StaticMesh|SetStaticMesh (LocalVariables|GetCurrentSplineComponent) Mesh)
        (Utilities|FlowControl|Sequence
          (:then_0
            (Utilities|IsValid Material
              (:Is%20Valid
                (Rendering|Material|SetMaterial _returnvalue 0 Material))))
          (:then_1
            (CallFunction|SetLocationfromSpline (LocalVariables|GetCurrentSplineComponent) (LocalVariables|GetCurrentPoint))
            (SplineMesh|SetForwardAxis (LocalVariables|GetCurrentSplineComponent) "X" true)))))
    (:then_1
      (return _localsplinecomponentlist))))

(fn Add_SplineMesh_Components_ForBlend
  (params (Evaluate%20WPO :type Boolean) (Mesh :type Static%20Mesh%20Object%20Reference) (HeightOffset :type Float%20%28single-precision%29) (AlphaFalloff :type Float%20%28single-precision%29) (BlendDistance :type Float%20%28single-precision%29) (BaseWidthAdd :type Float%20%28double-precision%29) (LeftSideColor :type Linear%20Color%20Structure :default "(R=1.000000,G=1.000000,B=1.000000,A=1.000000)") (RightSideColor :type Linear%20Color%20Structure :default "(R=1.000000,G=1.000000,B=1.000000,A=1.000000)"))
  (bind _localsplinecomponentlist (LocalVariables|GetLocalSplineComponentlist))
  (Utilities|FlowControl|Sequence
    (:then_0
      (for _index (range (Spline|GetNumberOfSplinePoints (Variables|Default|GetSpline)))
        (LocalVariables|SetCurrentPoint _index)
        (bind _returnvalue (AddComponent|NODE_AddSplineMeshComponent-1 :RelativeTransform (Math|Transform|MakeTransform "0, 0, 0")))
        (Rendering|LOD|SetEvaluateWorldPositionOffset _returnvalue Evaluate%20WPO :!class /Script/Engine.StaticMeshComponent)
        (Rendering|SetBoundsScale _returnvalue 3.0)
        (bind _output_get (LocalVariables|SetCurrentSplineComponent _returnvalue))
        (Utilities|Array|Add _localsplinecomponentlist _output_get)
        (Components|StaticMesh|SetStaticMesh (LocalVariables|GetCurrentSplineComponent) Mesh)
        (Utilities|FlowControl|Sequence
          (:then_0
            (bind _mid (CallFunction|MakeMaterialForBlendMesh _returnvalue HeightOffset AlphaFalloff BlendDistance BaseWidthAdd LeftSideColor RightSideColor)))
          (:then_1
            (bind (_scalestart _scaleend) (CallFunction|SetLocationfromSpline (LocalVariables|GetCurrentSplineComponent) (LocalVariables|GetCurrentPoint)))
            (Rendering|Material|SetScalarParameterValue _mid "ScaleStart" _scaleend :!class /Script/Engine.MaterialInstanceDynamic)
            (Rendering|Material|SetScalarParameterValue _mid "ScaleEnd" _scalestart :!class /Script/Engine.MaterialInstanceDynamic)
            (SplineMesh|SetForwardAxis (LocalVariables|GetCurrentSplineComponent) "X" true)))))
    (:then_1
      (return _localsplinecomponentlist))))

(fn AddEndBlendComponents
  (params (HeightOffset%20Input :type Float%20%28single-precision%29) (Error%20Margin%20input :type Float%20%28double-precision%29) (LeftSideColor :type Linear%20Color%20Structure :default "(R=1.000000,G=1.000000,B=1.000000,A=1.000000)") (RightSideColor :type Linear%20Color%20Structure :default "(R=1.000000,G=1.000000,B=1.000000,A=1.000000)") (Falloff :type Float%20%28single-precision%29))
  (bind _spline (Variables|Default|GetSpline))
  (bind _returnvalue (Spline|GetNumberOfSplinePoints _spline))
  (bind _returnvalue_2 (- _returnvalue 1))
  (bind _returnvalue_3 (Spline|GetTransformatSplinePoint _spline _returnvalue_2 "Local" false))
  (bind (_min _max) (Components|StaticMesh|GetLocalBounds (Variables|Default|GetEndComponent)))
  (bind _returnvalue_4 (Spline|GetTransformatSplinePoint _spline 0 "Local" false))
  (bind (_min_2 _max_2) (Components|StaticMesh|GetLocalBounds (Variables|Default|GetStartComponent)))
  (Utilities|FlowControl|Sequence
    (:then_0
      (Utilities|IsValid (Variables|Default|GetStaticMesh_Start)
        (:Is%20Valid
          (LocalVariables|SetStartMeshLength (- (.x _max_2) (.x _min_2)))
          (bind _staticmeshcomponent (CallFunction|AddEndBlendSpacerComponent _returnvalue_4 (Variables|GetHeightOffset%20Input) (Variables|GetError%20Margin%20input) (LocalVariables|GetStartMeshLength) (Variables|GetRightSideColor) (Variables|GetLeftSideColor) (Variables|GetFalloff)))
          (Transformation|AddLocalRotation _staticmeshcomponent "0,180.000000, 0" false false)
          (Utilities|Array|Add (LocalVariables|GetEndPieces) _staticmeshcomponent))))
    (:then_1
      (Utilities|IsValid (Variables|Default|GetEndComponent)
        (:Is%20Valid
          (LocalVariables|SetEndMeshLength (- (.x _max) (.x _min)))
          (bind _staticmeshcomponent_2 (CallFunction|AddEndBlendSpacerComponent _returnvalue_3 (Variables|GetHeightOffset%20Input) (Variables|GetError%20Margin%20input) (LocalVariables|GetEndMeshLength) (Variables|GetLeftSideColor) (Variables|GetRightSideColor) (Variables|GetFalloff)))
          (Utilities|Array|Add (LocalVariables|GetEndPieces) _staticmeshcomponent_2))))
    (:then_2
      (Utilities|FlowControl|Sequence
        (:then_0
          (bind _staticmeshcomponent_3 (CallFunction|AddEndBlendRoundComponent (Math|Transform|MakeTransform (+ (.location _returnvalue_4) (* (Math|Vector|Normalize (Math|Vector|NegateVector (Math|Vector|GetForwardVector (.rotation _returnvalue_4)))) (LocalVariables|GetStartMeshLength))) (.rotation _returnvalue_4) (.scale _returnvalue_4)) (Variables|GetHeightOffset%20Input) (Variables|GetError%20Margin%20input) (Variables|GetRightSideColor) (Variables|GetLeftSideColor) (Variables|GetFalloff)))
          (Utilities|Array|Add (LocalVariables|GetEndPieces) _staticmeshcomponent_3))
        (:then_1
          (bind _staticmeshcomponent_4 (CallFunction|AddEndBlendRoundComponent (Math|Transform|MakeTransform (+ (.location _returnvalue_3) (* (Math|Vector|Normalize (Math|Vector|GetForwardVector (.rotation _returnvalue_3))) (LocalVariables|GetEndMeshLength))) (.rotation _returnvalue_3) (.scale _returnvalue_3)) (Variables|GetHeightOffset%20Input) (Variables|GetError%20Margin%20input) (Variables|GetLeftSideColor) (Variables|GetRightSideColor) (Variables|GetFalloff)))
          (Transformation|AddLocalRotation _staticmeshcomponent_4 "0,180.000000, 0" false false)
          (Utilities|Array|Add (LocalVariables|GetEndPieces) _staticmeshcomponent_4))
        (:then_2
          (return (LocalVariables|GetEndPieces)))))))

(fn AddEndBlendRoundComponent
  (params (RelativeTransform :type Transform) (HeightOffset :type Float%20%28single-precision%29) (ErrorMargin :type Float%20%28double-precision%29) (LeftSideColor :type Linear%20Color%20Structure :default "(R=1.000000,G=1.000000,B=1.000000,A=1.000000)") (RightSideColor :type Linear%20Color%20Structure :default "(R=1.000000,G=1.000000,B=1.000000,A=1.000000)") (Falloff :type Float%20%28single-precision%29))
  (bind _staticmesh (Variables|Default|GetStaticMesh))
  (bind _returnvalue (StaticMesh|GetBounds _staticmesh))
  (bind (_origin _boxextent) (Math|BoxSphereBounds|BreakBoxSphereBounds _returnvalue))
  (bind _y (.y _boxextent))
  (bind _returnvalue_2 (* _y 1.0))
  (bind _basewidthmargin (Variables|TerrainBlend|GetBaseWidthMargin))
  (bind _returnvalue_3 (Utilities|Operators|Add _returnvalue_2 _basewidthmargin ErrorMargin))
  (bind _blend_distance (Variables|TerrainBlend|GetBlend%20Distance))
  (bind _returnvalue_4 (AddComponent|NODE_AddStaticMeshComponent-2 :RelativeTransform RelativeTransform))
  (Utilities|FlowControl|Sequence
    (:then_0
      (bind _returnvalue_5 (Rendering|Material|CreateDynamicMaterialInstance _returnvalue_4 0 "/Game/ArtTools/Procedural/LandscapeBlend/SplineMeshBlend/SplineBlendMesh_End.SplineBlendMesh_End" "None" :!class /Script/Engine.PrimitiveComponent))
      (Rendering|Material|SetScalarParameterValue _returnvalue_5 "BaseWidth" _returnvalue_3 :!class /Script/Engine.MaterialInstanceDynamic)
      (Rendering|Material|SetScalarParameterValue _returnvalue_5 "HeightOffset" HeightOffset :!class /Script/Engine.MaterialInstanceDynamic)
      (Rendering|Material|SetScalarParameterValue _returnvalue_5 "AlphaFalloff" Falloff :!class /Script/Engine.MaterialInstanceDynamic)
      (Rendering|Material|SetScalarParameterValue _returnvalue_5 "BlendDistance" _blend_distance :!class /Script/Engine.MaterialInstanceDynamic)
      (Rendering|Material|SetScalarParameterValue _returnvalue_5 "RadialWidthOffset" _returnvalue_3 :!class /Script/Engine.MaterialInstanceDynamic)
      (Rendering|Material|SetScalarParameterValue _returnvalue_5 "RadialWidth" _blend_distance :!class /Script/Engine.MaterialInstanceDynamic)
      (Rendering|Material|SetVectorParameterValue _returnvalue_5 "ColorLeft" LeftSideColor :!class /Script/Engine.MaterialInstanceDynamic)
      (Rendering|Material|SetVectorParameterValue _returnvalue_5 "ColorRight" RightSideColor :!class /Script/Engine.MaterialInstanceDynamic))
    (:then_1
      (Transformation|SetRelativeScale3D _returnvalue_4 (Math|Vector|MakeVector (/ (+ _returnvalue_3 _blend_distance) 1e+03) 1.0 1.0)))
    (:then_2
      (Transformation|AddLocalRotation _returnvalue_4 "0,180.000000, 0" false false))
    (:then_3
      (return _returnvalue_4))))

(fn AddEndBlendSpacerComponent
  (params (RelativeTransform :type Transform) (HeightOffset :type Float%20%28single-precision%29) (ErrorMargin :type Float%20%28double-precision%29) (Length :type Float%20%28double-precision%29) (LeftSideColor :type Linear%20Color%20Structure :default "(R=1.000000,G=1.000000,B=1.000000,A=1.000000)") (RightSideColor :type Linear%20Color%20Structure :default "(R=1.000000,G=1.000000,B=1.000000,A=1.000000)") (Falloff :type Float%20%28single-precision%29))
  (bind _staticmesh (Variables|Default|GetStaticMesh))
  (bind _returnvalue (StaticMesh|GetBounds _staticmesh))
  (bind (_origin _boxextent) (Math|BoxSphereBounds|BreakBoxSphereBounds _returnvalue))
  (bind _y (.y _boxextent))
  (bind _returnvalue_2 (* _y 1.0))
  (bind _basewidthmargin (Variables|TerrainBlend|GetBaseWidthMargin))
  (bind _returnvalue_3 (Utilities|Operators|Add _returnvalue_2 _basewidthmargin ErrorMargin))
  (bind _blend_distance (Variables|TerrainBlend|GetBlend%20Distance))
  (bind _returnvalue_4 (AddComponent|NODE_AddStaticMeshComponent-6 :RelativeTransform RelativeTransform))
  (Utilities|FlowControl|Sequence
    (:then_0
      (bind _returnvalue_5 (Rendering|Material|CreateDynamicMaterialInstance _returnvalue_4 0 "/Game/ArtTools/Procedural/LandscapeBlend/SplineMeshBlend/SplineBlendMesh.SplineBlendMesh" "None" :!class /Script/Engine.PrimitiveComponent))
      (Rendering|Material|SetScalarParameterValue _returnvalue_5 "BaseWidth" _returnvalue_3 :!class /Script/Engine.MaterialInstanceDynamic)
      (Rendering|Material|SetScalarParameterValue _returnvalue_5 "HeightOffset" HeightOffset :!class /Script/Engine.MaterialInstanceDynamic)
      (Rendering|Material|SetScalarParameterValue _returnvalue_5 "AlphaFalloff" Falloff :!class /Script/Engine.MaterialInstanceDynamic)
      (Rendering|Material|SetScalarParameterValue _returnvalue_5 "BlendDistance" _blend_distance :!class /Script/Engine.MaterialInstanceDynamic)
      (Rendering|Material|SetScalarParameterValue _returnvalue_5 "RadialWidthOffset" _returnvalue_3 :!class /Script/Engine.MaterialInstanceDynamic)
      (Rendering|Material|SetScalarParameterValue _returnvalue_5 "RadialWidth" _blend_distance :!class /Script/Engine.MaterialInstanceDynamic)
      (Rendering|Material|SetVectorParameterValue _returnvalue_5 "ColorLeft" LeftSideColor :!class /Script/Engine.MaterialInstanceDynamic)
      (Rendering|Material|SetVectorParameterValue _returnvalue_5 "ColorRight" RightSideColor :!class /Script/Engine.MaterialInstanceDynamic))
    (:then_1
      (Transformation|SetRelativeScale3D _returnvalue_4 (Math|Vector|MakeVector (/ Length 1e+03) 1.0 1.0)))
    (:then_3
      (return _returnvalue_4))))

(fn AddStartEnd
  (bind _spline (Variables|Default|GetSpline))
  (bind _returnvalue (Spline|GetNumberOfSplinePoints _spline))
  (bind _returnvalue_2 (- _returnvalue 1))
  (bind _staticmesh_start (Variables|Default|GetStaticMesh_Start))
  (bind _spline_2 (Variables|Default|GetSpline))
  (bind _staticmesh_end (Variables|Default|GetStaticMesh_End))
  (Utilities|FlowControl|Sequence
    (:then_0
      (Utilities|IsValid _staticmesh_start
        (:Is%20Valid
          (bind _returnvalue_3 (AddComponent|NODE_AddStaticMeshComponent-4 :RelativeTransform (Spline|GetTransformatSplinePoint _spline_2 0 "Local" false)))
          (Components|StaticMesh|SetStaticMesh _returnvalue_3 _staticmesh_start)
          (Transformation|SetWorldScale3D _returnvalue_3 (Math|Vector|MakeVector (.y (Spline|GetScaleatSplinePoint _spline_2 0)) (.y (Spline|GetScaleatSplinePoint _spline_2 0)) (.z (Spline|GetScaleatSplinePoint _spline_2 0))))
          (Variables|Default|SetStartComponent _returnvalue_3))))
    (:then_1
      (Utilities|IsValid _staticmesh_end
        (:Is%20Valid
          (bind _returnvalue_4 (AddComponent|NODE_AddStaticMeshComponent-5 :RelativeTransform (Spline|GetTransformatSplinePoint _spline _returnvalue_2 "Local" false)))
          (Components|StaticMesh|SetStaticMesh _returnvalue_4 _staticmesh_end)
          (Transformation|SetWorldScale3D _returnvalue_4 (Math|Vector|MakeVector (.y (Spline|GetScaleatSplinePoint _spline _returnvalue_2)) (.y (Spline|GetScaleatSplinePoint _spline _returnvalue_2)) (.z (Spline|GetScaleatSplinePoint _spline _returnvalue_2))))
          (Variables|Default|SetEndComponent _returnvalue_4))))))

(fn CaptureLandscapeHeight
  (bind (_min _max) (Utilities|Struct|BreakBox (StaticMesh|GetBoundingBox (Variables|Default|GetStaticMesh))))
  (bind (_x_max _x_min _y_max _y_min _size_x _size_y _size_z _origin_local _origin_world _scale_y_max) (CallFunction|GetSplineMeshBounds (Variables|Default|GetSpline)))
  ;[1 1 1] Capture Landscape
  (bind _outactors (Actor|GetAllActorsOfClass "/Script/Landscape.Landscape"))
  (bind _returnvalue (Rendering|CreateRenderTarget2D 512 512 "RTF_RG8"))
  (Variables|Default|SetRenderTarget_LandscapeHeight _returnvalue)
  (Landscape|Runtime|RenderHeightmap (Utilities|Array|Get(acopy) _outactors) (Math|Transform|MakeTransform _origin_world) (Math|Box2D|MakeBox2D (Math|Vector2D|MakeVector2D (* (/ (Math|Float|Max(Float) (Utilities|Operators|Add _size_x (Variables|TerrainBlend|GetPatchMargin) (* (- (.y _max) (.y _min)) _scale_y_max)) (Utilities|Operators|Add _size_y (Variables|TerrainBlend|GetPatchMargin) (* (- (.y _max) (.y _min)) _scale_y_max))) 2.0) -1.0) (* (/ (Math|Float|Max(Float) (Utilities|Operators|Add _size_x (Variables|TerrainBlend|GetPatchMargin) (* (- (.y _max) (.y _min)) _scale_y_max)) (Utilities|Operators|Add _size_y (Variables|TerrainBlend|GetPatchMargin) (* (- (.y _max) (.y _min)) _scale_y_max))) 2.0) -1.0)) (Math|Vector2D|MakeVector2D (/ (Math|Float|Max(Float) (Utilities|Operators|Add _size_x (Variables|TerrainBlend|GetPatchMargin) (* (- (.y _max) (.y _min)) _scale_y_max)) (Utilities|Operators|Add _size_y (Variables|TerrainBlend|GetPatchMargin) (* (- (.y _max) (.y _min)) _scale_y_max))) 2.0) (/ (Math|Float|Max(Float) (Utilities|Operators|Add _size_x (Variables|TerrainBlend|GetPatchMargin) (* (- (.y _max) (.y _min)) _scale_y_max)) (Utilities|Operators|Add _size_y (Variables|TerrainBlend|GetPatchMargin) (* (- (.y _max) (.y _min)) _scale_y_max))) 2.0))) "/Game/ArtTools/Procedural/LandscapeBlend/SplineMeshBlend/RenderTarget_Landscape.RenderTarget_Landscape"))

(fn ConformSplineToLandscape
  (bind _spline (Variables|Default|GetSpline))
  (for _index (range (Spline|GetNumberOfSplinePoints _spline))
    (bind (_outhits _returnvalue) (Collision|MultiLineTraceForObjects :Start (+ (Spline|GetLocationatSplinePoint _spline _index "World") "0, 0,5000.000000") :End (+ (Spline|GetLocationatSplinePoint _spline _index "World") "0, 0,-5000.000000") :ObjectTypes (Utilities|Array|MakeArray "ObjectTypeQuery1") :bTraceComplex false :DrawDebugType "ForOneFrame" :bIgnoreSelf true :TraceColor "(R=1.000000,G=0.000000,B=0.000000,A=1.000000)" :TraceHitColor "(R=0.000000,G=1.000000,B=0.000000,A=1.000000)" :DrawTime 1.0))
    (if _returnvalue
      (bind _array_element (Utilities|Array|ForEachLoopwithBreak _outhits)
        (:LoopBody
          (Utilities|Casting|CastToLandscape (.HitActor (Collision|BreakHitResult _array_element))
            (:then
              _)
            (:CastFailed
              (Utilities|Casting|CastToLandscapeProxy (.HitActor (Collision|BreakHitResult _array_element)))))
          (Spline|SetLocationatSplinePoint _spline _index (.Location (Collision|BreakHitResult _array_element)) "World" true)
          (LocalVariables|SetTempLength (Math|Vector|VectorLength (Spline|GetTangentatSplinePoint _spline _index "World")))
          (Spline|SetRotationatSplinePoint _spline _index (Math|Rotator|MakeRotationfromAxes (Math|Vector|GetForwardVector (Spline|GetRotationatSplinePoint _spline _index "World")) (Math|Vector|GetRightVector (Spline|GetRotationatSplinePoint _spline _index "World")) (.Normal (Collision|BreakHitResult _array_element))) "World" true)
          (Spline|SetTangentatSplinePoint _spline _index (* (Spline|GetTangentatSplinePoint _spline _index "World") (LocalVariables|GetTempLength)) "Local" true)))))
  (CallFunction|UpdateSpline)
  (CallFunction|UpdateStartandEndComponents)
  (return))

(fn GetClosestLandscape
  (bind (_distance _returnvalue) (Actor|FindNearestActor (Transformation|GetActorLocation) (LocalVariables|GetActorLandscape)))
  (Utilities|FlowControl|Sequence
    (:then_1
      (bind _outactors (Actor|GetAllActorsOfClass "/Script/Landscape.Landscape"))
      (for _array_element _outactors
        (Utilities|Array|Add (LocalVariables|GetActorLandscape) _array_element)))
    (:then_2
      (bind _outactors_2 (Actor|GetAllActorsOfClass "/Script/Landscape.LandscapeStreamingProxy"))
      (for _array_element_2 _outactors_2
        (Utilities|Array|Add (LocalVariables|GetActorLandscape) _array_element_2)))
    (:then_3
      (return _returnvalue))))

(fn GetSplineMeshBounds
  (params (Spline :type Spline%20Component%20Object%20Reference))
  (Utilities|FlowControl|Sequence
    (:then_0
      (for _index (range (Spline|GetNumberOfSplinePoints Spline))
        (LocalVariables|SetX_Max (Math|Float|Max(Float) (.x (Spline|GetLocationatSplinePoint Spline _index "Local")) (LocalVariables|GetX_Max)))
        (LocalVariables|SetX_Min (Math|Float|Min(Float) (LocalVariables|GetX_Min) (.x (Spline|GetLocationatSplinePoint Spline _index "Local"))))
        (LocalVariables|SetY_Max (Math|Float|Max(Float) (LocalVariables|GetY_Max) (.y (Spline|GetLocationatSplinePoint Spline _index "Local"))))
        (LocalVariables|SetY_Min (Math|Float|Min(Float) (LocalVariables|GetY_Min) (.y (Spline|GetLocationatSplinePoint Spline _index "Local"))))
        (LocalVariables|SetZ_Max (Math|Float|Max(Float) (.z (Spline|GetLocationatSplinePoint Spline _index "Local")) (LocalVariables|GetZ_Max)))
        (LocalVariables|SetZ_Min (Math|Float|Min(Float) (.z (Spline|GetLocationatSplinePoint Spline _index "Local")) (LocalVariables|GetZ_Min)))
        (LocalVariables|SetScale_Y_Max (Math|Float|Max(Float) (LocalVariables|GetScale_Y_Max) (.y (Spline|GetScaleatSplinePoint Spline _index))))))
    (:then_1
      (return (LocalVariables|GetX_Max) (LocalVariables|GetX_Min) (LocalVariables|GetY_Max) (LocalVariables|GetY_Min) (- (LocalVariables|GetX_Max) (LocalVariables|GetX_Min)) (- (LocalVariables|GetY_Max) (LocalVariables|GetY_Min)) (- (LocalVariables|GetZ_Max) (LocalVariables|GetZ_Min)) (Math|Vector|MakeVector (- (LocalVariables|GetX_Max) (/ (- (LocalVariables|GetX_Max) (LocalVariables|GetX_Min)) 2.0)) (- (LocalVariables|GetY_Max) (/ (- (LocalVariables|GetY_Max) (LocalVariables|GetY_Min)) 2.0)) (- (LocalVariables|GetZ_Max) (/ (- (LocalVariables|GetZ_Max) (LocalVariables|GetZ_Min)) 2.0))) (+ (Math|Vector|MakeVector (- (LocalVariables|GetX_Max) (/ (- (LocalVariables|GetX_Max) (LocalVariables|GetX_Min)) 2.0)) (- (LocalVariables|GetY_Max) (/ (- (LocalVariables|GetY_Max) (LocalVariables|GetY_Min)) 2.0)) (- (LocalVariables|GetZ_Max) (/ (- (LocalVariables|GetZ_Max) (LocalVariables|GetZ_Min)) 2.0))) (Transformation|GetActorLocation)) (LocalVariables|GetScale_Y_Max)))))

(fn GetStartEndBounds
  (Utilities|FlowControl|Sequence
    (:then_0
      (bind _startcomponent (Variables|Default|GetStartComponent)
        (:then
          (LocalVariables|SetStartLength (- (.x (.Max (Components|StaticMesh|GetLocalBounds _startcomponent))) (.x (Components|StaticMesh|GetLocalBounds _startcomponent)))))
        (:else
          (LocalVariables|SetStartLength 0.0))))
    (:then_1
      (bind _endcomponent (Variables|Default|GetEndComponent)
        (:then
          (LocalVariables|SetEndLength (- (.x (.Max (Components|StaticMesh|GetLocalBounds _endcomponent))) (.x (Components|StaticMesh|GetLocalBounds _endcomponent)))))
        (:else
          (LocalVariables|SetEndLength 0.0))))
    (:then_2
      (return (Math|Float|Max(Float) (LocalVariables|GetStartLength) (LocalVariables|GetEndLength))))))

(fn MakeMaterialforBlendMesh
  (params (Component :type Primitive%20Component%20Object%20Reference) (HeightOffset :type Float%20%28single-precision%29) (AlphaFalloff :type Float%20%28single-precision%29) (BlendDistance :type Float%20%28single-precision%29) (BaseWidthAdd :type Float%20%28double-precision%29) (LeftSideColor :type Linear%20Color%20Structure :default "(R=1.000000,G=1.000000,B=1.000000,A=1.000000)") (RightSideColor :type Linear%20Color%20Structure :default "(R=1.000000,G=1.000000,B=1.000000,A=1.000000)"))
  (bind (_origin _boxextent) (Math|BoxSphereBounds|BreakBoxSphereBounds (StaticMesh|GetBounds (Variables|Default|GetStaticMesh))))
  (bind _blendmaterial (LocalVariables|GetBlendMaterial))
  (bind _blendmaterial_2 (LocalVariables|GetBlendMaterial))
  (bind _returnvalue (Rendering|Material|CreateDynamicMaterialInstance "/Game/ArtTools/Procedural/LandscapeBlend/SplineMeshBlend/SplineBlendMesh.SplineBlendMesh" "None" "None" :!class /Script/Engine.KismetMaterialLibrary))
  (LocalVariables|SetBlendMaterial _returnvalue)
  (Rendering|Material|SetScalarParameterValue _returnvalue "BaseWidth" (Utilities|Operators|Add (* (.y _boxextent) 1.0) (Variables|TerrainBlend|GetBaseWidthMargin) BaseWidthAdd) :!class /Script/Engine.MaterialInstanceDynamic)
  (Rendering|Material|SetScalarParameterValue _blendmaterial_2 "HeightOffset" HeightOffset :!class /Script/Engine.MaterialInstanceDynamic)
  (Rendering|Material|SetScalarParameterValue _blendmaterial_2 "AlphaFalloff" AlphaFalloff :!class /Script/Engine.MaterialInstanceDynamic)
  (Rendering|Material|SetScalarParameterValue _blendmaterial_2 "BlendDistance" BlendDistance :!class /Script/Engine.MaterialInstanceDynamic)
  (Rendering|Material|SetVectorParameterValue _blendmaterial_2 "ColorLeft" LeftSideColor :!class /Script/Engine.MaterialInstanceDynamic)
  (Rendering|Material|SetVectorParameterValue _blendmaterial_2 "ColorRight" RightSideColor :!class /Script/Engine.MaterialInstanceDynamic)
  (Rendering|Material|SetMaterial Component 0 _blendmaterial)
  (return _blendmaterial))

(fn SetLocationFromSpline
  (params (SplineComponant :type Spline%20Mesh%20Component%20Object%20Reference) (CurrentPoint :type Integer))
  (bind _returnvalue (+ CurrentPoint 1))
  (bind _spline (Variables|Default|GetSpline))
  (bind (_location _tangent) (Spline|GetLocationandTangentatSplinePoint _spline _returnvalue "Local"))
  (bind (_location_2 _tangent_2) (Spline|GetLocationandTangentatSplinePoint _spline CurrentPoint "Local"))
  (bind _spline_2 (Variables|Default|GetSpline))
  (bind _returnvalue_2 (Spline|GetScaleatSplinePoint _spline_2 _returnvalue))
  (bind _y (.y _returnvalue_2))
  (bind _spline_3 (Variables|Default|GetSpline))
  (bind _returnvalue_3 (Spline|GetScaleatSplinePoint _spline_3 CurrentPoint))
  (bind _y_2 (.y _returnvalue_3))
  (bind _splinecomponant (Variables|GetSplineComponant))
  (bind _splinecomponant_2 (Variables|GetSplineComponant))
  (Utilities|FlowControl|Sequence
    (:then_0
      (LocalVariables|SetCurrenPointLocation _location_2)
      (LocalVariables|SetCurrentTangent _tangent_2)
      (LocalVariables|SetNextPointLocation _location)
      (LocalVariables|SetNextTangent _tangent)
      (SplineMesh|SetStartandEnd SplineComponant (LocalVariables|GetCurrenPointLocation) (Math|Vector|SelectVector (* (Math|Vector|Normalize (LocalVariables|GetCurrentTangent)) (* (Math|Vector|Distance(Vector) _location_2 _location) (Variables|Default|GetTension))) (LocalVariables|GetCurrentTangent) (Variables|Default|GetAutoSmoothTangent)) (LocalVariables|GetNextPointLocation) (Math|Vector|SelectVector (* (Math|Vector|Normalize (LocalVariables|GetNextTangent)) (* (Math|Vector|Distance(Vector) _location_2 _location) (Variables|Default|GetTension))) (LocalVariables|GetNextTangent) (Variables|Default|GetAutoSmoothTangent)) true :!class /Script/Engine.SplineMeshComponent)
      (SplineMesh|SetStartScale _splinecomponant_2 (Math|Vector2D|MakeVector2D _y_2 (.z _returnvalue_3)) true :!class /Script/Engine.SplineMeshComponent)
      (SplineMesh|SetEndScale _splinecomponant_2 (Math|Vector2D|MakeVector2D _y (.z _returnvalue_2)) true :!class /Script/Engine.SplineMeshComponent)
      (SplineMesh|SetStartRollRadians _splinecomponant (Math|Trig|DegreesToRadians (Spline|GetRollatSplinePoint _spline CurrentPoint "Local")) true)
      (SplineMesh|SetEndRollRadians _splinecomponant (Math|Trig|DegreesToRadians (Spline|GetRollatSplinePoint _spline _returnvalue "Local")) true))
    (:then_1
      (return _y_2 _y))))

(fn SetSplineMeshCollision
  (params (SplineComponent :type Spline%20Mesh%20Component%20Object%20Reference))
  (if (Variables|Default|GetUseCollision)
    (Collision|SetCollisionEnabled SplineComponent "QueryAndPhysics")
    (Collision|SetCollisionResponsetoChannel SplineComponent "ECC_GameTraceChannel11" "ECR_Block") ; Block foot step traces for sound
    (else
      (Collision|SetCollisionEnabled SplineComponent "NoCollision"))))

(fn SetStartEndComponentVisiblity
  (params (bNewVisibility :type Boolean))
  (bind _startcomponent (Variables|Default|GetStartComponent))
  (bind _endcomponent (Variables|Default|GetEndComponent))
  (Utilities|FlowControl|Sequence
    (:then_0
      (Utilities|IsValid _startcomponent
        (:Is%20Valid
          (Rendering|SetVisibility _startcomponent bNewVisibility false))))
    (:then_1
      (Utilities|IsValid _endcomponent
        (:Is%20Valid
          (Rendering|SetVisibility _endcomponent bNewVisibility false))))))

(fn SetupLandscapePatch
  (params (Coverage :type Vector%202D%20Structure) (NewLocation :type Vector))
  (bind _landscapetexturepatch (Variables|Default|GetLandscapeTexturePatch))
  (bind _resolutionout (LandscapePatch|GetInitResolutionfromLandscape _landscapetexturepatch 1.0))
  (Transformation|SetWorldLocation (Variables|Default|GetLandscapeTexturePatch) NewLocation false true)
  (LandscapePatch|SetUnscaledCoverage (Variables|Default|GetLandscapeTexturePatch) Coverage)
  (LandscapePatch|SetResolution _landscapetexturepatch _resolutionout)
  (LandscapePatch|SetHeightRenderTargetFormat _landscapetexturepatch "RTF_RGBA32f")
  (Development|PrintString (Utilities|String|ToString(Vector2d) _resolutionout)))

(fn SetVisibilityOnArray
  (params (Array :type Array%20of%20Spline%20Mesh%20Component%20Object%20References) (bNewVisibility :type Boolean))
  (for _array_element Array
    (Rendering|SetVisibility _array_element bNewVisibility false)))

(fn UpdateSpline
  (for (_array_element _array_index) (Variables|Default|GetAllSplineComponents)
    (CallFunction|SetLocationfromSpline _array_element _array_index)))

(fn UpdateStartAndEndComponents
  (bind _spline (Variables|Default|GetSpline))
  (bind _returnvalue (Spline|GetNumberOfSplinePoints _spline))
  (bind _returnvalue_2 (- _returnvalue 1))
  (bind _spline_2 (Variables|Default|GetSpline))
  (bind _startcomponent (Variables|Default|GetStartComponent))
  (bind _endcomponent (Variables|Default|GetEndComponent))
  (Utilities|FlowControl|Sequence
    (:then_0
      (Utilities|IsValid (Variables|Default|GetStaticMesh_Start)
        (:Is%20Valid
          (Transformation|SetRelativeTransform _startcomponent (Spline|GetTransformatSplinePoint _spline_2 0 "Local" false) false false)
          (Transformation|SetWorldScale3D _startcomponent (Math|Vector|MakeVector (.y (Spline|GetScaleatSplinePoint _spline_2 0)) (.y (Spline|GetScaleatSplinePoint _spline_2 0)) (.z (Spline|GetScaleatSplinePoint _spline_2 0)))))))
    (:then_1
      (Utilities|IsValid (Variables|Default|GetStaticMesh_End)
        (:Is%20Valid
          (Transformation|SetRelativeTransform _endcomponent (Spline|GetTransformatSplinePoint _spline _returnvalue_2 "Local" false) false false)
          (Transformation|SetWorldScale3D _endcomponent (Math|Vector|MakeVector (.y (Spline|GetScaleatSplinePoint _spline _returnvalue_2)) (.y (Spline|GetScaleatSplinePoint _spline _returnvalue_2)) (.z (Spline|GetScaleatSplinePoint _spline _returnvalue_2)))))))))

(fn ConstructionScript
  (bind _allsplinecomponents (CallFunction|AddSplineMeshComponents false (Variables|Default|GetStaticMesh)))
  (Variables|Default|SetAllSplineComponents _allsplinecomponents)
  (CallFunction|AddStartEnd)
  (LandscapePatch|SetHeightSourceMode (Variables|Default|GetLandscapeTexturePatch) "TextureBackedRenderTarget")
  (CallFunction|UpdateLandscapePatch))
