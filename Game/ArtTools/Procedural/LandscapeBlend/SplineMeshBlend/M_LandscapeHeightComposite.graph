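# Post-process material that composites a landscape height render target with captured object heights, in a spline mesh landscape-blend tool.
# Landscape height is decoded from 16-bit R/G channels and scaled by Landscape_ZScale.
# Object and min-height captures, combined via floored alpha and Max, become terrain-relative heights using camera and terrain Z.
# A final Max keeps the higher surface; a SpiralBlur branch and 500-unit lift feed nothing shown.
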Material /Game/ArtTools/Procedural/LandscapeBlend/SplineMeshBlend/M_LandscapeHeightComposite
graph {
 properties {
  bAllowNegativeEmissiveColor=true
  blendableOutputAlpha=true
  materialDomain=MD_PostProcess
  pixelDepthOffsetMode=PDOM_Legacy
  shadingModel=MSM_Unlit
  usedShadingModels=Unlit
 }
 nodes [
  n1 = TextureSample() {samplerType=SAMPLERTYPE_LinearColor, texture=&/RenderTarget_Landscape}
  n2 = Multiply(A=Add(A=Multiply(A=@n1.R) {constB=256}, B=@n1.G)) {constB=255}
  n3 = Multiply(A=Divide(A=Subtract(A=@n2) {constB=32768}) {constB=128}, B=ScalarParameter() {defaultValue=100, parameterName=Landscape_ZScale})
  Add(A=@n3) {constB=500, desc=This Should Lift the Terrain 500 units}
  ComponentMask(None=MaterialFunctionCall!(
    TextureObject=TextureObject() {samplerType=SAMPLERTYPE_LinearColor, texture=&/RenderTarget_Object}
    UV
    Distance=ScalarParameter() {defaultValue=0.01, parameterName=Distance}
    Distance Steps
    Radial Steps
    Radial Offset
    Kernel Power
   ) {materialFunction=/Engine/Functions/Engine_MaterialFunctions01/Texturing/SpiralBlur-Texture}.Result) {r=true}
  Constant() {r=1}
  n4 = ScalarParameter() {parameterName=Z_PositionCamera}
  n5 = TextureSampleParameter2D() {
   parameterName=RenderTargetObject
   samplerType=SAMPLERTYPE_LinearColor
   texture=&RenderTarget_Object
  }
  n6 = TextureSampleParameter2D() {
   parameterName=RenderTargetMinHeight
   samplerType=SAMPLERTYPE_LinearColor
   texture=&/RenderTarget_MinHeight
  }
  n7 = TextureSampleParameter2D() {
   parameterName=RenderTargetAlpha
   samplerType=SAMPLERTYPE_LinearColor
   texture=&/RenderTarget_MINAlpha
  }
  n8 = Max(A=LinearInterpolate(A=@n5.R, B=@n6.R, Alpha=Floor(None=@n7.R)), B=@n5.R)
  Max(A=@n3, B=Subtract(A=Subtract(A=@n4, B=@n8), B=ScalarParameter() {parameterName=Z_PositionTerrain}))
  Reroute(None=@n7.R)
 ]
}
