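# Post-process material that, by its name, generates landscape paint weights for spline mesh blending.
# Gradient-texture noise on scaled UVs forms red and green masks, each with its own inversion and strength parameters and a shared cheap-contrast adjustment.
# Each mask multiplies its channel of the RenderTargetAlpha texture and a weight parameter, then the two are summed.
# The Random offset does not feed the noise, and no material output connection appears.
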Material /Game/ArtTools/Procedural/LandscapeBlend/SplineMeshBlend/M_Landscape_Weight_Paint
graph {
 properties {
  bAllowNegativeEmissiveColor=true
  blendableOutputAlpha=true
  materialDomain=MD_PostProcess
  pixelDepthOffsetMode=PDOM_Legacy
  shadingModel=MSM_Unlit
  usedShadingModels=Unlit
 }
 nodes [
  n1 = AppendVector(A=TextureCoordinate(), B=Constant())
  n2 = ScalarParameter() {parameterName=Random}
  Add(A=@n1, B=AppendVector(A=AppendVector(A=@n2, B=@n2), B=@n2))
  n3 = Multiply(A=@n1, B=ScalarParameter() {defaultValue=2, parameterName=WorldSize})
  n4 = Multiply(B=ScalarParameter() {defaultValue=10, parameterName=Size}) {constA=1000}
  n5 = Saturate(None=Noise!(World Position=Divide(A=@n3, B=@n4), FilterWidth) {bTurbulence=false, levelScale=1, levels=0, noiseFunction=NOISEFUNCTION_GradientTex, outputMin=0, scale=5})
  n6 = OneMinus(None=@n5)
  n7 = LinearInterpolate(
   B=LinearInterpolate(A=@n5, B=@n6, Alpha=ScalarParameter() {parameterName=isInvertedRed})
   Alpha=ScalarParameter() {parameterName=NoiseStrengthRed}
  ) {constA=1}
  n8 = ScalarParameter() {defaultValue=2, parameterName=Contrast}
  n9 = ScalarParameter() {defaultValue=1, parameterName=Red}
  n10 = TextureSampleParameter2D() {
   parameterName=RenderTargetAlpha
   samplerType=SAMPLERTYPE_LinearColor
   texture=&/RenderTarget_MINAlpha
  }
  n11 = LinearInterpolate(
   B=LinearInterpolate(A=@n5, B=@n6, Alpha=ScalarParameter() {parameterName=isInvertedGreen})
   Alpha=ScalarParameter() {parameterName=NoiseStrengthGreen}
  ) {constA=1}
  n12 = ScalarParameter() {defaultValue=1, parameterName=Green}
  n13 = Add(
   A=Multiply(
    A=MaterialFunctionCall(In=@n7, Contrast=@n8) {materialFunction=&CheapContrast}.Result
    B=Multiply(A=@n9, B=@n10.R)
   )
   B=Multiply(
    A=MaterialFunctionCall(In=@n11, Contrast=@n8) {materialFunction=&CheapContrast}.Result
    B=Multiply(A=@n12, B=@n10.G)
   )
  )
  Ceil(None=@n13)
  Constant() {r=1}
  Multiply(A=@n7, B=Constant() {r=1})
  Multiply(A=@n13, B=@n8)
  Saturate(None=@n13)
 ]
}
