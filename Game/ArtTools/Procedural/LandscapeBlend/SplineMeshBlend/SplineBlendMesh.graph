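# Unlit spline-mesh material for landscape blending, building a two-sided colour and mask band along a spline.
# UV X splits ColorLeft/ColorRight and side masks across a soft centre, faded by vertex red or, with IsEnd, radial pivot distance.
# Vertex colours drive offset math: blue picks sideways direction, green interpolates start/end width scale, red extends the edge by BlendDistance.
# Further terms add an edge drop-down or height offset.
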Material /Game/ArtTools/Procedural/LandscapeBlend/SplineMeshBlend/SplineBlendMesh
graph {
 properties {bUsedWithSplineMeshes=true, pixelDepthOffsetMode=PDOM_Legacy, shadingModel=MSM_Unlit, usedShadingModels=Unlit}
 nodes [
  n1 = VectorParameter() {parameterName=ColorLeft, defaultValue=<rgba 1 1 1 1>}
  n2 = Subtract(A=ComponentMask(None=TextureCoordinate()) {r=true}) {constB=0.5}
  n3 = ScalarParameter() {defaultValue=235.7024, parameterName=BaseWidth}
  n4 = Divide(A=ScalarParameter() {defaultValue=100, parameterName=CenterBlend}, B=@n3)
  n5 = Divide(A=@n4)
  n6 = Saturate(None=Divide(A=Add(A=@n2, B=@n5), B=@n4))
  n7 = MaterialFunctionCall() {materialFunction=&ObjectPivotPoint}
  n8 = Multiply(A=Constant3Vector() {constant=<rgba 0 0 1 0>}, B=ScalarParameter() {parameterName=HeightOffset})
  n9 = Subtract(
   A=Distance(A=WorldPosition().XYZ, B=Add(A=@n7.Object Pivot Location, B=@n8))
   B=ScalarParameter() {defaultValue=499.4214, parameterName=RadialWidthOffset}
  )
  n10 = ScalarParameter() {defaultValue=2178.2173, parameterName=RadialWidth}
  n11 = StaticSwitchParameter(True=Saturate(None=Divide(A=@n9, B=@n10)), False=VertexColor().R) {parameterName=IsEnd}
  n12 = Saturate(None=Power(Base=OneMinus(None=@n11), Exp=ScalarParameter() {defaultValue=8, parameterName=AlphaFalloff}))
  n13 = VectorParameter() {parameterName=ColorRight, defaultValue=<rgba 1 1 1 1>}
  n14 = Multiply(A=@n2) {constB=-1}
  n15 = Saturate(None=Divide(A=Add(A=@n14, B=@n5), B=@n4))
  Add(
   A=Multiply(A=LinearInterpolate(A=@n1.RGB, Alpha=@n6) {constB=0}, B=@n12)
   B=Multiply(A=LinearInterpolate(A=@n13.RGB, Alpha=@n15) {constB=0}, B=@n12)
  )
  n16 = VertexColor()
  n17 = Transform(None=LinearInterpolate(A=Constant3Vector() {constant=<rgba 1 0 0 0>}, B=Constant3Vector() {constant=<rgba -1 0 0 0>}, Alpha=@n16.B))
  n18 = Normalize(VectorInput=@n17)
  n19 = LinearInterpolate(
   A=ScalarParameter() {defaultValue=1, parameterName=ScaleStart}
   B=ScalarParameter() {defaultValue=1, parameterName=ScaleEnd}
   Alpha=VertexColor().G
  )
  n20 = Multiply(A=Constant() {r=-100}, B=@n19)
  n21 = Multiply(A=@n19, B=@n3)
  n22 = ScalarParameter() {defaultValue=2067.251, parameterName=BlendDistance}
  n23 = Add(A=Multiply(A=@n18, B=Add(A=@n20, B=@n21)), B=LinearInterpolate(A=Constant3Vector(), B=Multiply(A=@n18, B=@n22), Alpha=@n16.R))
  n24 = Power(Base=@n16.R, Exp=ScalarParameter() {defaultValue=1, parameterName=EdgeFalloff})
  n25 = Constant3Vector() {constant=<rgba 0 0 -1 0>}
  Add(A=@n23, B=Multiply(A=Multiply(A=@n24, B=@n25), B=ScalarParameter() {parameterName=DropDownDistance}))
  Add(A=@n23, B=Multiply(A=Constant3Vector() {constant=<rgba 0 0 1 0>}, B=ScalarParameter() {parameterName=HeightOffset}))
  Constant()
  LinearInterpolate(A=Constant3Vector(), B=@n18, Alpha=VertexColor().A)
  Multiply(A=@n17, B=Constant3Vector() {constant=<rgba 1 1 0 0>})
  n26 = ScalarParameter() {defaultValue=1, parameterName=LeftSideMask}
  Multiply(
   A=LinearInterpolate(A=ScalarParameter() {defaultValue=1, parameterName=RightSideMask}, Alpha=@n15)
   B=Multiply(A=LinearInterpolate(A=@n26, Alpha=@n6), B=@n12)
  )
  Subtract(A=@n2, B=@n5) {constB=0.5}
  Subtract(A=@n14, B=@n5) {constB=0.5}
 ]
}
